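; Player duel bounty ability that tracks the target for the owning "Phoebe" player and decides how the duel ends.
; On the server, a TargetFound gameplay event supplies target and quest, then it binds to target death, quest failure and timeout.
; These map to success (Phoebe's squad/team kill), poached or failed states, passed to AdditionalUrgentQuestEndedLogic, an empty hook here.
; Ending unbinds delegates and removes configured effects, using pawn references set only on death.
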(event Ability|EventFailedToActivatePassiveAbility
  (params (ActorInfo :type Gameplay%20Ability%20Actor%20Info%20Structure))
  (|CallParentFunction ActorInfo))

(event Ability|EventActivateAbility
  (bind _returnvalue (CallFunction|GetAvatarasFortPlayerPawn))
  (bind _targetpawnnotphoebe (Variables|Default|GetTargetPawnNotPhoebe))
  (bind _returnvalue_2 (Class|FortPlayerPawn|GetTransientQuestsComponent _targetpawnnotphoebe))
  ;[0 0 0] Activate ability
  (|CallParentFunction)
  (Ability|HasAuthority
    (:True
      ;; Wait for payload to come through with target and instigator info to pass to Phoebe
      (bind (_asynctaskproxy _payload) (Ability|Tasks|WaitGameplayEvent "(TagName=\"Event.PlayerDuel.TargetFound\")" 0 true)
        (:EventReceived
          (bind _asfort_quest_item_definition (Utilities|Casting|CastToFortQuestItemDefinition (.OptionalObject (Utilities|Struct|BreakGameplayEventData _payload)))
            (:then
              (Variables|Default|SetParameterQuestDefinition _asfort_quest_item_definition)
              (if (== _returnvalue (.Instigator (Utilities|Struct|BreakGameplayEventData _payload)))
                (Variables|Default|SetOwningPawnPhoebe _returnvalue)
                ;;[0.1927 0.1927 0.1927 0.5] Track when the target player from the payload is eliminated and when the target's quest fails
                (bind _asfort_player_pawn (Utilities|Casting|CastToFortPlayerPawn (.Target (Utilities|Struct|BreakGameplayEventData _payload)))
                  (:then
                    (bind _output_get (Variables|Default|SetTargetPawnNotPhoebe _asfort_player_pawn))
                    (bind _returnvalue_3 (Fortnite|Utilities|AttemptGetPlayerStateforPawn _output_get))
                    (Utilities|IsValid _returnvalue_3
                      (:Is%20Valid
                        (EventDispatchers|AssignOnPawnDiedDelegate _returnvalue_3 (EventDispatchers|CreateEvent :!event Custom|OnTargetDied))
                        (bind _asfort_urgent_quest_item_definition (Utilities|Casting|CastToFortUrgentQuestItemDefinition (Variables|Default|GetParameterQuestDefinition)))
                        (Variables|Default|SetTargetUrgentQuest _asfort_urgent_quest_item_definition)
                        (EventDispatchers|BindEventtoOnUrgentQuestTimeExpired _returnvalue_2 (EventDispatchers|CreateEvent :!event Custom|UrgentQuestTimeExpired))
                        (EventDispatchers|BindEventtoOnUrgentQuestFailed _returnvalue_2 (EventDispatchers|CreateEvent :!event Custom|TargetQuestFailed)))
                      (:Is%20Not%20Valid
                        (Ability|EndAbility))))
                  (:CastFailed
                    (Ability|EndAbility)))
                (else
                  _)))
            (:CastFailed
              _))
          (Ability|EndAbility))))))

(event Custom|OnTargetDied
  (params (PawnDamageInfo :type Pawn%20Damage%20Data%20Structure%20%28by%20ref%29))
  ;[0 0 0] When a pawn is eliminated, determine which team was successful and send poached verbs for the other team
  (CallFunction|SetPawnsfromPayload PawnDamageInfo)
  (if (Variables|Default|GetbAnyPawnsValid)
    (if (or (Squad|OnSameSquad (Variables|Default|GetDamagerPawnAthena) (Variables|Default|GetOwningPawnAthena)) (Team|OnSameTeam (Variables|Default|GetDamagerPawnAthena) (Variables|Default|GetOwningPawnAthena)))
      (Development|PrintString "target player is elim'd by phoebe's team - call success" true true "(R=1.000000,G=0.000000,B=0.682086,A=1.000000)" 5.0)
      (CallFunction|AdditionalUrgentQuestEndedLogic "NewEnumerator5")
      (else
        (Development|PrintString "target player not elim'd by phoebe or team - call poached" true true "(R=1.000000,G=0.000000,B=0.682086,A=1.000000)" 5.0)
        (CallFunction|AdditionalUrgentQuestEndedLogic "NewEnumerator4")))
    (Ability|EndAbility)
    (else
      _)))

(event Custom|AdditionalUrgentQuestEndedLogic
  (params (EnumDuelQuestState :type EnumDuelStateTrackingPhoebe%20Enum :default "NewEnumerator0")))

(event Custom|UrgentQuestTimeExpired
  (params (UrgentQuestData :type Urgent%20Quest%20Data%20Structure%20%28by%20ref%29))
  ;[0 0 0] When target's quest expires, run failure logic on phoebe
  (if (== (Utilities|Struct|BreakUrgentQuestData UrgentQuestData) (Utilities|Struct|BreakUrgentQuestData (Class|FortUrgentQuestItemDefinition|GetUrgentQuestData (Variables|Default|GetTargetUrgentQuest))))
    (Development|PrintString "Target quest timed out - execute fail logic" true true "(R=1.000000,G=0.000000,B=0.682086,A=1.000000)" 5.0)
    (CallFunction|AdditionalUrgentQuestEndedLogic "NewEnumerator0")
    (Ability|EndAbility)))

(event Ability|EventOnEndAbility
  (params (bWasCancelled :type Boolean))
  (bind _owningpawnathena (Variables|Default|GetOwningPawnAthena))
  (bind _returnvalue (Ability|GetFortAbilitySystemComponent _owningpawnathena))
  (bind _targetpawnathena (Variables|Default|GetTargetPawnAthena))
  (bind _returnvalue_2 (Class|FortPlayerPawn|GetTransientQuestsComponent _targetpawnathena))
  ;[0 0 0] End Ability - Unbind all events and cancel this GE
  (|CallParentFunction bWasCancelled)
  (Utilities|IsValid _targetpawnathena
    (:Is%20Valid
      (EventDispatchers|UnbindEventfromOnUrgentQuestFailed _returnvalue_2 (EventDispatchers|CreateEvent :!event Custom|TargetQuestFailed))
      (EventDispatchers|UnbindEventfromOnUrgentQuestTimeExpired _returnvalue_2 (EventDispatchers|CreateEvent :!event Custom|UrgentQuestTimeExpired))
      (bind _returnvalue_3 (Fortnite|Utilities|AttemptGetPlayerStateforPawn (Variables|Default|GetTargetPawnAthena)))
      (Utilities|IsValid _returnvalue_3
        (:Is%20Valid
          (EventDispatchers|UnbindEventfromOnPawnDiedDelegate _returnvalue_3 (EventDispatchers|CreateEvent :!event Custom|OnTargetDied))
          (if (Utilities|GameplayTagContainerNotEmpty (Variables|Default|GetGameplayEffectsToRemove))
            (Utilities|IsValid _returnvalue
              (:Is%20Valid
                (GameplayEffects|RemoveActiveEffectswithAppliedTags _returnvalue (Variables|Default|GetGameplayEffectsToRemove))))))))))

(event Custom|TargetQuestFailed
  (params (PlayerStateAthena :type Fort%20Player%20State%20Athena%20Object%20Reference) (UrgentQuestData :type Urgent%20Quest%20Data%20Structure%20%28by%20ref%29))
  ;[0 0 0] When the target's quest fails, determine the state Pheobe should execute
  (if (and (== (Utilities|Struct|BreakUrgentQuestData UrgentQuestData) (Utilities|Struct|BreakUrgentQuestData (Class|FortUrgentQuestItemDefinition|GetUrgentQuestData (Variables|Default|GetTargetUrgentQuest)))) (not (FortPawn|IsDead (Variables|Default|GetTargetPawnNotPhoebe))))
    (Development|PrintString "Target failed quest" true true "(R=1.000000,G=0.000000,B=0.682086,A=1.000000)" 5.0)
    (CallFunction|AdditionalUrgentQuestEndedLogic "NewEnumerator4")
    (Ability|EndAbility)
    (else
      (Development|PrintString "quest instance does not match or target was elim'd" true true "(R=1.000000,G=0.000000,B=0.682086,A=1.000000)" 5.0))))

(fn SetPawnsFromPayload
  (params (PawnDamageInfo :type Pawn%20Damage%20Data%20Structure))
  (bind (_damagedactor _damage _instigatedby) (Utilities|Struct|BreakPawnDamageData PawnDamageInfo))
  (bind _returnvalue (Ability|GetAvatarActorfromActorInfo))
  (bind _returnvalue_2 (Pawn|GetControlledPawn _instigatedby))
  (bind (_asfort_player_pawn_athena _bsuccess) (Utilities|Casting|CastToFortPlayerPawnAthena _returnvalue_2))
  (bind (_asplayer_pawn_athena _bsuccess_2) (Utilities|Casting|CastToPlayerPawn_Athena _damagedactor))
  (bind (_asfort_player_pawn_athena_2 _bsuccess_3) (Utilities|Casting|CastToFortPlayerPawnAthena _returnvalue))
  ;[0 0 0] Helper function to set references to the target, damager, and ability owner as Athena Pawns
  ;;[0.1198 0.1198 0.1198] Check that damaged pawn, owner, and instigator are valid Athena Pawns
  (Utilities|IsValid _instigatedby
    (:Is%20Valid
      (if (Math|Boolean|ANDBoolean _bsuccess_2 _bsuccess_3 _bsuccess)
        ;;[0.151 0.151 0.151] Store references as Athena Pawns
        (Variables|Default|SetTargetPawnAthena _asplayer_pawn_athena)
        (Variables|Default|SetOwningPawnAthena _asfort_player_pawn_athena_2)
        (Variables|Default|SetDamagerPawnAthena _asfort_player_pawn_athena)
        (Variables|Default|SetbAnyPawnsValid true)
        (return)
        (else
          _)))
    (:Is%20Not%20Valid
      _))
  (Variables|Default|SetbAnyPawnsValid false))
