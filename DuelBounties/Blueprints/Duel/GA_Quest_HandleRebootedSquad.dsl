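; Gameplay ability in Duel Bounties that grants an urgent-quest gameplay effect to squadmates when they are rebooted.
; On activation it gathers the owner's squad, binding dead members to pawn possession and living ones to pawn death, which then binds possession.
; On same-team possession it waits briefly, applies the effect to the rebooted pawn and sends a squadmate-rebooted event to the owner.
; Ending the ability unbinds the delegates and clears the squad list.
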(event Ability|EventActivateAbility
  ;[0 0 0] Activate ability
  (|CallParentFunction)
  (Variables|Default|SetOwningPlayerPawn (CallFunction|GetAvatarasFortPlayerPawn))
  (CallFunction|StartListeningforSquadmatesDied))

(event Custom|OnPawnPossessed
  (params (Pawn :type Pawn%20Object%20Reference))
  ;[0 0 0] When a squadmate is rebooted, remove the JobBoard active effect as this isn't shown in the UI
  (bind _asfort_player_pawn_athena (Utilities|Casting|CastToFortPlayerPawnAthena Pawn))
  (Variables|Default|SetRebootedPawn _asfort_player_pawn_athena)
  (if (Team|OnSameTeam (Variables|Default|GetRebootedPawn) (Variables|Default|GetOwningPlayerPawn))
    (Utilities|FlowControl|Delay 1.0) ; Give time for the player to "arrive" after being rebooted
    (GameplayEffects|ApplyGameplayEffectToSelf :self (Ability|GetFortAbilitySystemComponent (Variables|Default|GetRebootedPawn)) :GameplayEffectClass (Variables|Default|GetGrantUrgentQuestOnReboot_GE) :Level 0.0)
    (Ability|SendGameplayEventtoActor (Variables|Default|GetOwningPlayerPawn) (Variables|Default|GetSquadmateRebootedEventTag) (Utilities|Struct|MakeGameplayEventData (Variables|Default|GetSquadmateRebootedEventTag) 0 (Variables|Default|GetRebootedPawn)))))

(event Custom|OnPawnDied
  (params (PawnDamageInfo :type Pawn%20Damage%20Data%20Structure%20%28by%20ref%29))
  ;[0 0 0] When a pawn that was alive when the duel was accepted dies, bind to when they are rebooted
  (bind _asfort_player_pawn_athena (Utilities|Casting|CastToFortPlayerPawnAthena (Utilities|Struct|BreakPawnDamageData PawnDamageInfo)))
  (if (Team|OnSameTeam _asfort_player_pawn_athena (Variables|Default|GetOwningPlayerPawn))
    (Game|Player|BindEventtoOnPlayerPawnPossessed (Fortnite|GetFortPlayerControllerfromActor _asfort_player_pawn_athena) (EventDispatchers|CreateEvent :!event Custom|OnPawnPossessed))))

(event Ability|EventOnEndAbility
  (params (bWasCancelled :type Boolean))
  ;[0 0 0] On end ability - unbind events from squadmates
  (|CallParentFunction bWasCancelled)
  (for _array_element (Variables|Default|GetSquadPlayerStates)
    (EventDispatchers|UnbindEventfromOnPawnDiedDelegate _array_element (EventDispatchers|CreateEvent :!event Custom|OnPawnDied))
    (Utilities|IsValid (Fortnite|GetFortPlayerControllerfromActor _array_element)
      (:Is%20Valid
        (Game|Player|UnbindEventfromOnPlayerPawnPossessed (Fortnite|GetFortPlayerControllerfromActor _array_element) (EventDispatchers|CreateEvent :!event Custom|OnPawnPossessed)))) ; Make sure this is valid before trying to unbind
  )
  (Utilities|Array|Clear (Variables|Default|GetSquadPlayerStates)))

(event Ability|EventFailedToActivatePassiveAbility
  (params (ActorInfo :type Gameplay%20Ability%20Actor%20Info%20Structure))
  ;[0 0 0] Failed to activate passive ability
  (|CallParentFunction ActorInfo))

(fn StartListeningForSquadmatesDied
  (bind _owningplayerpawn (Variables|Default|GetOwningPlayerPawn))
  ;[0 0 0] Get the owner's squadmates
  (Utilities|Array|Clear (Variables|Default|GetSquadPlayerStates))
  (bind _returnvalue (Fortnite|Utilities|AttemptGetPlayerStateforPawn _owningplayerpawn))
  (bind _asfort_player_state_athena (Utilities|Casting|CastToFortPlayerStateAthena _returnvalue))
  (bind _returnvalue_2 (Class|FortGameStateAthena|GetSquadMemberPlayerStates (Fortnite|Athena|GameState|GetGameStateAthenafromContext _owningplayerpawn) _asfort_player_state_athena false))
  (bind _output_get (Variables|Default|SetSquadPlayerStates _returnvalue_2))
  (for _array_element _output_get
    (if (Class|FortPlayerStateZone|IsPlayerDead _array_element)
      (Game|Player|BindEventtoOnPlayerPawnPossessed (Fortnite|GetFortPlayerControllerfromActor _array_element) (EventDispatchers|CreateEvent :!event OnPawnPossessed))
      (else
        (EventDispatchers|BindEventtoOnPawnDiedDelegate _array_element (EventDispatchers|CreateEvent :!event OnPawnDied)))))
  (return))
