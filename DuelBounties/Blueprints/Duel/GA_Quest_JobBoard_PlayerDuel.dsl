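; Gameplay ability for the job board player duel quest, tracking a target player from the event payload until that player is eliminated.
; On activation it adds a radius tracker on the target for the owning player and binds the target's death delegate.
; Kills by the owner's squad, or converted teammates, trigger quest success; others broadcast Poached, which also goes to the victim's squad if tagged instigator.
; HandleInstigatorLogic is an empty hook.
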(event Ability|EventActivateAbilityFromEvent
  (params (EventData :type Gameplay%20Event%20Data%20Structure%20%28by%20ref%29))
  (bind (_eventtag _instigator _target) (Utilities|Struct|BreakGameplayEventData EventData))
  ;[0 0 0] Track when the target player from the payload is eliminated and call to handle instigator logic
  (|CallParentFunction EventData)
  (bind _asfort_player_controller (Utilities|Casting|CastToFortPlayerController _target))
  (bind _output_get (Variables|Default|SetTargetPawn (Game|Player|GetPlayerPawn _asfort_player_controller)))
  (CallFunction|MakeTrackingRadiusforTarget _output_get)
  (bind _returnvalue (Fortnite|Utilities|AttemptGetPlayerStateforPawn _output_get))
  (EventDispatchers|AssignOnPawnDiedDelegate _returnvalue (EventDispatchers|CreateEvent :!event Custom|OnPawnDied_Delegate))
  (if (GameplayTags|HasMatchingGameplayTag (CallFunction|GetAvatarasFortPlayerPawn) (Variables|Default|GetInstigatorTag))
    (CallFunction|HandleInstigatorLogic)))

(event Custom|OnPawnDied_Delegate
  (params (PawnDamageInfo :type Pawn%20Damage%20Data%20Structure%20%28by%20ref%29))
  ;[0 0 0] When a pawn is eliminated, determine which team was successful and send poached verbs for the other team
  (CallFunction|SetPawnsfromPayload PawnDamageInfo)
  (if (Variables|Default|GetbAnyPawnsValid)
    (if (or (Squad|OnSameSquad (Variables|Default|GetDamagerPawnAthena) (Variables|Default|GetOwningPawnAthena)) (and (Team|OnSameTeam (Variables|Default|GetDamagerPawnAthena) (Variables|Default|GetOwningPawnAthena)) (GameplayTags|HasMatchingGameplayTag (Variables|Default|GetDamagerPawnAthena) "(TagName=\"Athena.Gameplay.Status.Converted\")")))
      (Utilities|FlowControl|Delay 0.5)
      (CallFunction|HandleDuelSuccess)
      (else
        (CallFunction|HandleDuelPoached)))
    (CallFunction|BroadcastPoachedtoTargetSquad)
    (RadiusTracker|RemoveTrackingforActor (RadiusTracker|GetGameStateRadiusTrackerComponent) (Variables|Default|GetTargetPawnAthena))))

(event Custom|BroadcastPoachedToTargetSquad
  (bind _targetpawnathena (Variables|Default|GetTargetPawnAthena))
  ;[0 0 0] Broadcast the poached message to squadmates of the target eliminated
  (Utilities|IsValid _targetpawnathena
    (:Is%20Valid
      (if (GameplayTags|HasMatchingGameplayTag _targetpawnathena "(TagName=\"Granted.Athena.Duel.Instigator\")")
        (bind _returnvalue (Fortnite|Utilities|AttemptGetPlayerStateforPawn _targetpawnathena))
        (bind _asfort_player_state_athena (Utilities|Casting|CastToFortPlayerStateAthena _returnvalue))
        (bind _returnvalue_2 (Class|FortGameStateAthena|GetSquadMemberPlayerStates (Fortnite|Athena|GameState|GetGameStateAthenafromContext _targetpawnathena) _asfort_player_state_athena false))
        (for _array_element _returnvalue_2
          (CustomVerb|BroadcastCustomVerbMessage (CustomVerb|MakeCustomVerbMessage _array_element "(TagName=\"CustomVerb.JobBoard.Poached\")")))))))

(event Ability|EventOnEndAbility
  (params (bWasCancelled :type Boolean))
  (bind _returnvalue (CallFunction|GetAvatarasFortPlayerPawn))
  (bind _returnvalue_2 (Ability|GetAbilitySystemComponent _returnvalue))
  ;[0 0 0] On End Ability - clean up binds
  (|CallParentFunction bWasCancelled)
  (bind _returnvalue_3 (Fortnite|Utilities|AttemptGetPlayerStateforPawn (Variables|Default|GetTargetPawn)))
  (Utilities|IsValid _returnvalue_3
    (:Is%20Valid
      (EventDispatchers|UnbindEventfromOnPawnDiedDelegate _returnvalue_3 (EventDispatchers|CreateEvent :!event Custom|OnPawnDied_Delegate)))
    (:Is%20Not%20Valid
      _))
  (Utilities|IsValid _returnvalue_2
    (:Is%20Valid
      (GameplayEffects|RemoveActiveEffectswithGrantedTags _returnvalue_2 (Variables|Default|GetDuelInstigatorTag)))))

(event Custom|SetPawnsFromPayload
  (params (PawnDamageInfo :type Pawn%20Damage%20Data%20Structure))
  (bind (_damagedactor _damage _instigatedby) (Utilities|Struct|BreakPawnDamageData PawnDamageInfo))
  (bind _returnvalue (Ability|GetAvatarActorfromActorInfo))
  (bind _returnvalue_2 (Pawn|GetControlledPawn _instigatedby))
  (bind (_asfort_player_pawn_athena _bsuccess) (Utilities|Casting|CastToFortPlayerPawnAthena _returnvalue_2))
  (bind (_asplayer_pawn_athena _bsuccess_2) (Utilities|Casting|CastToPlayerPawn_Athena _damagedactor))
  (bind (_asfort_player_pawn_athena_2 _bsuccess_3) (Utilities|Casting|CastToFortPlayerPawnAthena _returnvalue))
  ;[0 0 0] Helper event to set references to the target, damager, and ability owner as Athena Pawns
  ;;[0.1198 0.1198 0.1198] Check that damaged pawn, owner, and instigator are valid Athena Pawns
  (Utilities|IsValid _instigatedby
    (:Is%20Valid
      (if (Math|Boolean|ANDBoolean _bsuccess_2 _bsuccess_3 _bsuccess)
        ;;[0.151 0.151 0.151] Store references as Athena Pawns
        (Variables|Default|SetTargetPawnAthena _asplayer_pawn_athena)
        (Variables|Default|SetOwningPawnAthena _asfort_player_pawn_athena_2)
        (Variables|Default|SetDamagerPawnAthena _asfort_player_pawn_athena)
        (Variables|Default|SetbAnyPawnsValid true)
        (else
          _)))
    (:Is%20Not%20Valid
      _))
  (Variables|Default|SetbAnyPawnsValid false))

(event Custom|HandleDuelPoached
  (bind _returnvalue (CallFunction|GetAvatarasFortPlayerPawn))
  (bind _returnvalue_2 (Ability|GetAbilitySystemComponent _returnvalue))
  ;[0 0 0] Send quest verbs for each team
  (CustomVerb|BroadcastCustomVerbMessage (CustomVerb|MakeCustomVerbMessage (CallFunction|GetAvatarasFortPlayerPawn) "(TagName=\"CustomVerb.JobBoard.Poached\")"))
  (Utilities|IsValid _returnvalue_2
    (:Is%20Valid
      (GameplayEffects|RemoveActiveEffectswithTags _returnvalue_2 (Variables|Default|GetCosmeticOverrideGeAssetTag)))))

(event Custom|HandleDuelSuccess
  (bind _returnvalue (CallFunction|GetAvatarasFortPlayerPawn))
  (bind _returnvalue_2 (Ability|GetAbilitySystemComponent _returnvalue))
  ;[0 0 0] Send quest verbs for each team
  (Utilities|FlowControl|Delay 0.2)
  (Default|CallOnQuestSuccess)
  (CustomVerb|BroadcastCustomVerbMessage (CustomVerb|MakeCustomVerbMessage (CallFunction|GetAvatarasFortPlayerPawn) "(TagName=\"CustomVerb.PlayerDuel.TargetEliminated\")"))
  (Utilities|IsValid _returnvalue_2
    (:Is%20Valid
      (GameplayEffects|RemoveActiveEffectswithTags _returnvalue_2 (Variables|Default|GetCosmeticOverrideGeAssetTag)))))

(event Custom|HandleInstigatorLogic)

(fn MakeTrackingRadiusForTarget
  (params (TargetPawn :type Fort%20Player%20Pawn%20Object%20Reference))
  ;[0 0 0] Make tracking radius for the duel target
  (bind _returnvalue (RadiusTracker|MakeRadiusTrackingInitializer (LocalVariables|GetMaxCenterOffset) (LocalVariables|GetTrackingMinRadius) (LocalVariables|GetTrackingMaxRadius) (LocalVariables|GetTimeBetweenUpdates) (LocalVariables|GetRadiusAmountToShrink) "(R=1.000000,G=0.456411,B=0.000000,A=0.400000)" (LocalVariables|GetCircle%20Drawing%20Material) TargetPawn (Utilities|Array|MakeArray (Fortnite|GetFortPlayerControllerfromActor (CallFunction|GetAvatarasFortPlayerPawn)))))
  (Fortnite|AddActorTrackingInfo (RadiusTracker|GetGameStateRadiusTrackerComponent) _returnvalue))
