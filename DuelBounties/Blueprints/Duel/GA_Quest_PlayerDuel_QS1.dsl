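; Transient quest ability for player duel bounties that gives each participant an urgent quest and a shrinking tracking radius on the duel target.
; On the authority it waits for a target-found event, then binds death, timer-expiry and squad pawn-change delegates.
; Eliminations resolve as success, failed or poached, while disconnects force a poach; ability end broadcasts verbs and cleans up.
; The instigator's copy forwards remaining duel time to rebooted squadmates through gameplay events.
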(event Ability|EventActivateAbility
  ;[0 0 0] Activate ability and failed to activate passive ability
  (|CallParentFunction)
  (Variables|Default|SetIsQuestCompletedSuccessfully false)
  (Variables|DuelBountiesVariables|SetbAnyPlayerDisconnected false))

(event Ability|EventFailedToActivatePassiveAbility
  (params (ActorInfo :type Gameplay%20Ability%20Actor%20Info%20Structure))
  ;[0 0 0] Activate ability and failed to activate passive ability
  (|CallParentFunction ActorInfo))

(event EventOnPostAbilityInitialization
  ;[0 0 0] Post ability init - save urgent quest and wait for a duelist to be found
  (Ability|HasAuthority
    (:True
      ;; Save off urgent quest and owning player, and stop the urgent quest event if it is running until the target is found
      (bind _asfort_urgent_quest_item_definition (Utilities|Casting|CastToFortUrgentQuestItemDefinition (Variables|FortTransientQuestGameplayAbility|GetQuestItemDefinition))
        (:then
          (Variables|DuelBountiesVariables|SetUrgentQuestReference _asfort_urgent_quest_item_definition)
          (CallFunction|StopUrgentQuestEvent (Class|FortUrgentQuestItemDefinition|GetUrgentQuestData _asfort_urgent_quest_item_definition))
          (bind _output_get (Variables|ParticipantInfo|SetOwningPlayerPawn (CallFunction|GetAvatarasFortPlayerPawn)))
          (bind (_asynctaskproxy _payload) (Ability|Async|WaitGameplayEventToActor _output_get (Variables|DuelBountiesVariables|GetTargetFoundEventTag) true)
            (:EventReceived
              (Variables|Default|SetEventPayload _payload)
              ;; Save target pawn and make tracking radius for the target
              (bind _asfort_player_pawn (Utilities|Casting|CastToFortPlayerPawn (.Instigator (Utilities|Struct|BreakGameplayEventData _payload)))
                (:then
                  (Variables|ParticipantInfo|SetDuelInstigator _asfort_player_pawn)
                  (bind _asfort_player_pawn_2 (Utilities|Casting|CastToFortPlayerPawn (.Target (Utilities|Struct|BreakGameplayEventData _payload)))
                    (:then
                      (Variables|ParticipantInfo|SetTargetPawn _asfort_player_pawn_2)
                      (CallFunction|MakeTrackingRadiusforTarget (Variables|ParticipantInfo|GetTargetPawn) (ScalableFloat|GetValueatLevel (Variables|HotfixableVariables|GetCircleMinRadius)) (ScalableFloat|GetValueatLevel (Variables|HotfixableVariables|GetCircleMaxRadius)) (ScalableFloat|GetValueatLevel (Variables|HotfixableVariables|GetCircleAmountToShrink)))
                      ;; Save off player states for target and owner
                      (bind (_targetplayerstate _instigatorplayerstate) (CallFunction|GetPlayerStatesforParticipants (Variables|ParticipantInfo|GetDuelInstigator) (Variables|ParticipantInfo|GetTargetPawn)))
                      (Variables|ParticipantInfo|SetTarget%20Player%20State _targetplayerstate)
                      (Variables|ParticipantInfo|SetDuel%20Instigator%20Player%20State _instigatorplayerstate)
                      ;;[0.1927 0.1927 0.1927 0.5] Track when any instigators from the payload are eliminated, when the quest expires, or when any participant disconnects
                      (bind _output (CallFunction|GetAllParticipatingSquadStates))
                      (Variables|ParticipantInfo|SetParticipantSquadStates _output)
                      (CallFunction|HandleParticipantStateEvents false)
                      ;; Pass the info to the urgent quest. If the owner is also the instigator, set up logic for handling re-granting the quest for rebooted squadmates
                      (CallFunction|PassInfotoUrgentQuest (== (Utilities|Struct|BreakGameplayEventData (Variables|Default|GetEventPayload)) (Variables|Default|GetStartDuelForRebootedPlayerTag)))
                      (if (CallFunction|OwnerIsDuelInstigator)
                        (CallFunction|HandleSquadmateRebooted)))
                    (:CastFailed
                      _)))
                (:CastFailed
                  _))
              (Ability|EndAbility))))
        (:CastFailed
          (Ability|EndAbility))))))

(event Custom|PassInfoToUrgentQuest
  (params (bRebootedSquadmate :type Boolean))
  (bind (_eventtag _visualstyletag _totaleventtime _bdisplaytimer _bquestcompleteontimeexpiration _bquickintro _bquickoutro _brewardpersist _bdisplaymultistepexpanded _eventmaintitle _eventtitle _eventsubtitle _eventsubtitlesecondary _eventfailuremessage _eventdescription _eventdescriptiongamepad _eventdescriptiontouch _eventdescriptionoverrides _eventdescriptionoverridetag _sortpriority _bhidequestprogressbar _bdisplayallobjectivesindividually _questprogresstextformat _breverseprogressbar _bshowbountythreatinformation _bshowbountypriceanddistance _bdisplaystartedeventui _bissharedurgentquest _displayplayer _acceptingplayer _socialavatarbrushptr _bountypriceimage _alerticon _charactericon _specialeventstartedsound) (Utilities|Struct|BreakUrgentQuestData (Class|FortUrgentQuestItemDefinition|GetUrgentQuestData (Variables|DuelBountiesVariables|GetUrgentQuestReference))))
  (bind _eventpayload (Variables|Default|GetEventPayload))
  (bind (_eventtag_2 _instigator _target _optionalobject _optionalobject2 _contexthandle _instigatortags _targettags _eventmagnitude) (Utilities|Struct|BreakGameplayEventData _eventpayload))
  ;[0 0 0] Pass info to urgent quest
  (if bRebootedSquadmate
    (if (> _eventmagnitude 1.5)
      (Variables|Default|SetEventTimeToUse _eventmagnitude))
    (else
      (Variables|Default|SetEventTimeToUse (ScalableFloat|GetValueatLevel (Variables|HotfixableVariables|GetDuelBountiesTime)))))
  ;; Pass in dynamic information from the found player event payload to make the urgent quest
  (CallFunction|StartUrgentQuestEvent (Utilities|Struct|MakeUrgentQuestData :EventTag _eventtag :VisualStyleTag "(GameplayTags=)" :TotalEventTime (Variables|Default|GetEventTimeToUse) :bDisplayTimer true :bQuestCompleteOnTimeExpiration false :bQuickIntro false :bQuickOutro false :bRewardPersist false :bDisplayMultiStepExpanded true :EventTitle _eventtitle :EventSubtitle (Utilities|Text|FormatText _eventsubtitle (Class|FortPlayerControllerGameplay|GetSettingsFilteredPlayerName (Utilities|Casting|CastToFortPlayerControllerGameplay (Fortnite|GetFortPlayerControllerfromActor (Variables|ParticipantInfo|GetOwningPlayerPawn))) (Variables|ParticipantInfo|GetTarget%20Player%20State))) :EventSubtitleSecondary (Utilities|Text|FormatText _eventsubtitlesecondary (Fortnite|GetPlayerNameSafe (Variables|ParticipantInfo|GetDuelInstigator))) :EventDescription (Utilities|Text|FormatText _eventsubtitle (Class|FortPlayerControllerGameplay|GetSettingsFilteredPlayerName (Utilities|Casting|CastToFortPlayerControllerGameplay (Fortnite|GetFortPlayerControllerfromActor (Variables|ParticipantInfo|GetOwningPlayerPawn))) (Variables|ParticipantInfo|GetTarget%20Player%20State))) :EventDescriptionOverrideTag "(TagName=\"\")" :SortPriority 0 :bHideQuestProgressBar false :bDisplayAllObjectivesIndividually true :QuestProgressTextFormat "LOCGEN_FORMAT_ORDERED(INVTEXT(\"`{{0}} / `{{1}}\"), INVTEXT(\"CurrentProgress\"), INVTEXT(\"MaxProgress\"))" :bReverseProgressBar true :bShowBountyThreatInformation false :bShowBountyPriceAndDistance false :bDisplayStartedEventUI true :bIsSharedUrgentQuest false :DisplayPlayer (Variables|ParticipantInfo|GetTarget%20Player%20State) :AcceptingPlayer (Variables|ParticipantInfo|GetDuel%20Instigator%20Player%20State) :SocialAvatarBrushPtr (Class|FortPlayerStateZone|GetSocialAvatarBrush (Variables|ParticipantInfo|GetTarget%20Player%20State) true) :SpecialEventStartedSound _specialeventstartedsound :AlertIconBorderColor "(R=1.000000,G=0.456411,B=0.000000,A=1.000000)" :AlertIconBrushSize "(X=0.000000,Y=0.000000)" :bOverrideWidgetPrimaryColor false :OverriddenWidgetPrimaryColor "(R=1.000000,G=0.456411,B=0.000000,A=1.000000)" :ObjectiveCategory "Default")))

(event Custom|OnParticipantDied
  (params (PawnDamageInfo :type Pawn%20Damage%20Data%20Structure%20%28by%20ref%29))
  (bind _damagerpawnathena (Variables|ParticipantInfo|GetDamagerPawnAthena))
  ;[0 0 0] When a pawn is eliminated, determine which team was successful and send poached verbs for the other team
  (CallFunction|SetPawnsfromPayload PawnDamageInfo)
  (if (and (or (FortPawn|IsDead (Variables|ParticipantInfo|GetOwningPlayerPawn)) (Health|IsDBNO (Variables|ParticipantInfo|GetOwningPlayerPawn))) (CallFunction|OwnerIsDuelInstigator))
    (CallFunction|FailQuestVisualLogic)
    (elif (Variables|ParticipantInfo|GetbAnyPawnsValid)
      (bind (_damageronowningteam _eliminatedpawnonowningteam) (CallFunction|CheckTeamStatusforDamager))
      (if _damageronowningteam
        ;; Owning team completed the duel - send "success" message and complete the quest
        (CallFunction|MakeUrgentQuestMessage (Variables|DuelBountiesVariables|GetSuccessText) true)
        (Utilities|FlowControl|Delay 0.8)
        (CallFunction|HandleDuelSuccess)
        (Variables|Default|SetIsQuestCompletedSuccessfully true)
        (elif _eliminatedpawnonowningteam
          ;; If the damager is the target of the owning team's duel or if the player elim'd was on the owning team, send "failed" message; otherwise, send "poached" message
          (Utilities|IsValid _damagerpawnathena
            (:Is%20Valid
              (if (== _damagerpawnathena (Utilities|Casting|CastToFortPlayerPawnAthena (Variables|ParticipantInfo|GetTargetPawn)))
                (CallFunction|MakeUrgentQuestMessage (Variables|DuelBountiesVariables|GetFailedText))
                (else
                  (CallFunction|MakeUrgentQuestMessage (Variables|DuelBountiesVariables|GetPoachedText))))
              (Utilities|FlowControl|Delay 0.8)
              (CallFunction|HandleDuelPoachedorFailed)
              (Variables|Default|SetIsQuestCompletedSuccessfully false))
            (:Is%20Not%20Valid
              (CallFunction|FailQuestVisualLogic)))
          (else
            _))))))

(event Custom|FailQuestVisualLogic
  ;[0 0 0] When a pawn is eliminated, determine which team was successful and send poached verbs for the other team
  ;; If the damager is the target of the owning team's duel or if the player elim'd was on the owning team, send "failed" message; otherwise, send "poached" message
  (CallFunction|MakeUrgentQuestMessage (Variables|DuelBountiesVariables|GetFailedText))
  (Utilities|FlowControl|Delay 0.8)
  (CallFunction|HandleDuelPoachedorFailed)
  (Variables|Default|SetIsQuestCompletedSuccessfully false))

(event Custom|ManuallyPoachQuest
  ;[0 0 0] When a pawn is eliminated, determine which team was successful and send poached verbs for the other team
  ;; If the damager is the target of the owning team's duel or if the player elim'd was on the owning team, send "failed" message; otherwise, send "poached" message
  (CallFunction|MakeUrgentQuestMessage (Variables|DuelBountiesVariables|GetPoachedText))
  (Utilities|FlowControl|Delay 0.8)
  (CallFunction|HandleDuelPoachedorFailed)
  (Variables|Default|SetIsQuestCompletedSuccessfully false))

(event Ability|EventOnEndAbility
  (params (bWasCancelled :type Boolean))
  ;[0 0 0] On end ability
  (CallFunction|RemoveTrackingRadiusforTarget)
  (if (Variables|DuelBountiesVariables|GetbAnyPlayerDisconnected)
    (CallFunction|HandleParticipantStateEvents true)
    (CallFunction|ServerEndQuest)
    (elif (GameplayTags|IsGameplayTagValid (Variables|DuelBountiesVariables|GetPoachOrSuccessVerbTag))
      (CustomVerb|BroadcastCustomVerbMessage (CustomVerb|MakeCustomVerbMessage (Variables|ParticipantInfo|GetOwningPlayerPawn) (Variables|DuelBountiesVariables|GetPoachOrSuccessVerbTag)))
      ;; Unbind events from the participant player states if they are valid
      (Utilities|FlowControl|Delay 3.0) ; Wait for other duel abilities to finish before executing unbind logic
      (CallFunction|HandleParticipantStateEvents true)
      (CallFunction|ServerEndQuest))))

(event Custom|ServerEndQuest
  ;[0 0 0] Server End Quest with Success/Failure and Clean up
  (Utilities|FlowControl|DoOnce false
    (:Completed
      (if (Variables|Default|GetIsQuestCompletedSuccessfully)
        (CallFunction|CompleteQuest)
        (else
          (CallFunction|FailQuest))))))

(event Custom|HandleUrgentQuestTimeExpired
  (params (UrgentQuestData :type Urgent%20Quest%20Data%20Structure%20%28by%20ref%29))
  ;[0 0 0] HELPER/ORGANIZATIONAL EVENTS
  ;;[0.07292 0.07292 0.07292] Time expired - Make sure quest is updated to show correct visual fail state
  (if (== (Utilities|Struct|BreakUrgentQuestData UrgentQuestData) (Variables|DuelBountiesVariables|GetSpecialEventTag))
    (CallFunction|FailQuestVisualLogic)))

(event Custom|HandleSquadmateRebooted
  ;[0 0 0] HELPER/ORGANIZATIONAL EVENTS
  ;;[0.07292 0.07292 0.07292] Handle cases for when a squadmate dies and is rebooted
  ;;; Keep track of the duel time
  (bind _returnvalue (Utilities|Time|SetTimerbyEvent (delegate Custom|DuelTimeUp) (ScalableFloat|GetValueatLevel (Variables|HotfixableVariables|GetDuelBountiesTime)) false true))
  (Variables|Default|SetDuelTimerHandle _returnvalue)
  ;;; When a player is rebooted, send an event to them to start their copy of the quest with the elapsed time
  (bind (_asynctaskproxy _payload) (Ability|Async|WaitGameplayEventToActor (Variables|ParticipantInfo|GetDuelInstigator) (Variables|Default|GetSquadmateRebootedTag))
    (:EventReceived
      (bind _returnvalue_2 (Utilities|Time|SetTimerbyEvent (delegate Custom|PostDelayReboot) 1.5 false true)) ; Give time for the player's quest to init and wait to send the event
      (Variables|Default|SetRebootTimerDelayHandle _returnvalue_2))) ; Can be triggered multiple times
)

(event Custom|DuelTimeUp
  (bind _dueltimerhandle (Variables|Default|GetDuelTimerHandle))
  ;[0 0 0] HELPER/ORGANIZATIONAL EVENTS
  ;;[0.07292 0.07292 0.07292] Handle cases for when a squadmate dies and is rebooted
  ;;; Keep track of the duel time
  (if (Utilities|Time|IsValidTimerHandle _dueltimerhandle)
    (Utilities|Time|ClearandInvalidateTimerbyHandle _dueltimerhandle)))

(event Custom|PostDelayReboot
  (bind (_eventtag _instigator _target) (Utilities|Struct|BreakGameplayEventData (Variables|Default|GetEventPayload)))
  (bind _reboottimerdelayhandle (Variables|Default|GetRebootTimerDelayHandle))
  ;[0 0 0] HELPER/ORGANIZATIONAL EVENTS
  ;;[0.07292 0.07292 0.07292] Handle cases for when a squadmate dies and is rebooted
  ;;; When a player is rebooted, send an event to them to start their copy of the quest with the elapsed time
  (Ability|SendGameplayEventtoActor (.Target (Utilities|Struct|BreakGameplayEventData (stale Custom|HandleSquadmateRebooted _payload))) (Variables|Default|GetStartDuelForRebootedPlayerTag) (Utilities|Struct|MakeGameplayEventData (Variables|Default|GetStartDuelForRebootedPlayerTag) _instigator _target 0 0 "()" "(GameplayTags=)" "(GameplayTags=)" (- (ScalableFloat|GetValueatLevel (Variables|HotfixableVariables|GetDuelBountiesTime)) (Utilities|Time|GetTimerElapsedTimebyHandle (Variables|Default|GetDuelTimerHandle)))))
  (if (Utilities|Time|IsValidTimerHandle _reboottimerdelayhandle)
    (Utilities|Time|ClearandInvalidateTimerbyHandle _reboottimerdelayhandle)))

(event Custom|DisplayUrgentQuestMessage
  (params (UrgentQuestData :type Urgent%20Quest%20Data%20Structure) (bSuccess :type Boolean))
  ;[0 0 0] HELPER/ORGANIZATIONAL EVENTS
  ;;[0.07451 0.07451 0.07451] Display Urgent Quest Message
  (Class|AthenaHUDContext|ShowUrgentQuestEndedMessage (BlueprintContext|GetAthenaHUDContext) UrgentQuestData bSuccess))

(event Custom|HandleParticipantStateEvents
  (params (bShouldUnbindEvents :type Boolean))
  (bind _owningplayerpawn (Variables|ParticipantInfo|GetOwningPlayerPawn))
  (bind _returnvalue (Class|FortPlayerPawn|GetTransientQuestsComponent _owningplayerpawn))
  (bind _owningplayerpawn_2 (Variables|ParticipantInfo|GetOwningPlayerPawn))
  (bind _returnvalue_2 (Class|FortPlayerPawn|GetTransientQuestsComponent _owningplayerpawn_2))
  (bind _duel_instigator_player_state (Variables|ParticipantInfo|GetDuel%20Instigator%20Player%20State))
  (bind _target_player_state (Variables|ParticipantInfo|GetTarget%20Player%20State))
  (bind _participantsquadstates (Variables|ParticipantInfo|GetParticipantSquadStates))
  (bind _participantsquadstates_2 (Variables|ParticipantInfo|GetParticipantSquadStates))
  ;[0 0 0] HELPER/ORGANIZATIONAL EVENTS
  ;;[0.1927 0.1927 0.1927 0.5] Track when any instigators from the payload are eliminated or when the quest expires. Call with bool checked to clean up these events
  (if bShouldUnbindEvents
    ;;; Unbind events from the participant player states if they are valid
    (Utilities|IsValid _target_player_state
      (:Is%20Valid
        (EventDispatchers|UnbindEventfromOnPawnDiedDelegate _target_player_state (EventDispatchers|CreateEvent :!event Custom|OnParticipantDied))
        (Utilities|IsValid _duel_instigator_player_state
          (:Is%20Valid
            (EventDispatchers|UnbindEventfromOnPawnDiedDelegate _duel_instigator_player_state (EventDispatchers|CreateEvent :!event Custom|OnParticipantDied))
            (Utilities|IsValid _returnvalue
              (:Is%20Valid
                (EventDispatchers|UnbindEventfromOnUrgentQuestTimeExpired _returnvalue (EventDispatchers|CreateEvent :!event Custom|HandleUrgentQuestTimeExpired))
                (if (Utilities|Array|IsNotEmpty _participantsquadstates)
                  (for _array_element _participantsquadstates
                    (Utilities|IsValid (Fortnite|GetFortPlayerControllerfromActor _array_element)
                      (:Is%20Valid
                        (Pawn|UnbindEventfromOnPossessedPawnChanged (Fortnite|GetFortPlayerControllerfromActor _array_element) (EventDispatchers|CreateEvent :!event Custom|OnParticipantPawnChanged))))))))))))
    (else
      (EventDispatchers|AssignOnPawnDiedDelegate (Variables|ParticipantInfo|GetTarget%20Player%20State) (EventDispatchers|CreateEvent :!event Custom|OnParticipantDied))
      (EventDispatchers|AssignOnPawnDiedDelegate (Variables|ParticipantInfo|GetDuel%20Instigator%20Player%20State) (EventDispatchers|CreateEvent :!event Custom|OnParticipantDied))
      (Utilities|IsValid _returnvalue_2
        (:Is%20Valid
          (EventDispatchers|BindEventtoOnUrgentQuestTimeExpired _returnvalue_2 (EventDispatchers|CreateEvent :!event Custom|HandleUrgentQuestTimeExpired))
          (if (Utilities|Array|IsNotEmpty _participantsquadstates_2)
            (CallFunction|BindtoParticipantPawnChanges _participantsquadstates_2)))))))

(event Custom|OnParticipantPawnChanged
  (params (OldPawn :type Pawn%20Object%20Reference) (NewPawn :type Pawn%20Object%20Reference))
  (bind _participantplayerpawns (Variables|ParticipantInfo|GetParticipantPlayerPawns))
  ;[0 0 0] HELPER/ORGANIZATIONAL EVENTS
  ;; Helper events called when a given player disconnects so we can clean up the quest properly for other participants that could still be in-game
  ;;; Find a reference to the player who changed pawns from the owning squad player states
  (bind _asfort_player_pawn (Utilities|Casting|CastToFortPlayerPawn OldPawn))
  (if (.ReturnValue (Utilities|Map|Find _participantplayerpawns _asfort_player_pawn))
    (CallFunction|HandleDisconnectedPlayer NewPawn (Utilities|Map|Find _participantplayerpawns _asfort_player_pawn))))

(event Custom|HandleDisconnectedPlayer
  (params (NewPawn :type Pawn%20Object%20Reference) (PlayerState :type Fort%20Player%20State%20Athena%20Object%20Reference))
  (bind _returnvalue (RadiusTracker|GetGameStateRadiusTrackerComponent))
  ;[0 0 0] HELPER/ORGANIZATIONAL EVENTS
  ;; Helper events called when a given player disconnects so we can clean up the quest properly for other participants that could still be in-game
  ;;; Determine if the player changed states because they disconnected. If so, poach the quest
  (Utilities|IsValid NewPawn
    (:Is%20Not%20Valid
      (Utilities|IsValid PlayerState
        (:Is%20Valid
          (if (not (Class|FortPlayerStateZone|IsPlayerDead PlayerState))
            (Variables|DuelBountiesVariables|SetbAnyPlayerDisconnected true)
            (Utilities|IsValid _returnvalue
              (:Is%20Valid
                (RadiusTracker|RemoveTrackingforActorbyTrackingKey _returnvalue (Variables|ParticipantInfo|GetTargetPawn) (Variables|DuelBountiesVariables|GetRadiusTrackingGroupKey)))
              (:Is%20Not%20Valid
                _))
            (CallFunction|ManuallyPoachQuest)))) ; Confirm the player state passed in exists
    )) ; Confirm player wasn't rebooted to a valid pawn when changed
)

(fn BindToParticipantPawnChanges
  (params (AllParticipantPlayerStates :type Array%20of%20Fort%20Player%20State%20Athena%20Object%20References))
  ;[0 0 0] Used to determine if and when a player disconnects so that the quest can be poached accordingly
  (Utilities|Map|Clear (Variables|ParticipantInfo|GetParticipantPlayerPawns))
  (for _array_element AllParticipantPlayerStates
    ;; Find the pawn per player state and save it in a map so we can look it up later
    (Utilities|IsValid (Fortnite|GetFortPlayerControllerfromActor _array_element)
      (:Is%20Valid
        (Utilities|Map|Add (Variables|ParticipantInfo|GetParticipantPlayerPawns) (PlayerState|GetCurrentPawn _array_element) _array_element)
        (Pawn|BindEventtoOnPossessedPawnChanged (Fortnite|GetFortPlayerControllerfromActor _array_element) (EventDispatchers|CreateEvent :!event OnParticipantPawnChanged))))))

(fn BroadcastPoachedToTargetSquad)

(fn CheckTeamStatusForDamager
  (bind _owningpawnathena (Variables|ParticipantInfo|GetOwningPawnAthena))
  (bind _eliminatedpawnathena (Variables|ParticipantInfo|GetEliminatedPawnAthena))
  (bind _returnvalue (Squad|OnSameSquad _eliminatedpawnathena _owningpawnathena))
  ;[0 0 0] Helper function for determining if the damager is on the owning player's team
  (Utilities|IsValid (Variables|ParticipantInfo|GetDamagerPawnAthena)
    (:Is%20Valid
      (return (or (Squad|OnSameSquad (Variables|ParticipantInfo|GetDamagerPawnAthena) (Variables|ParticipantInfo|GetOwningPawnAthena)) (and (Team|OnSameTeam (Variables|ParticipantInfo|GetDamagerPawnAthena) (Variables|ParticipantInfo|GetOwningPawnAthena)) (GameplayTags|HasMatchingGameplayTag (Variables|ParticipantInfo|GetDamagerPawnAthena) "(TagName=\"Athena.Gameplay.Status.Converted\")"))) _returnvalue))
    (:Is%20Not%20Valid
      (return false _returnvalue))))

(fn GetAllParticipatingSquadStates
  ;[0 0 0] Helper function for tracking all duel participants
  (for _array_element (Utilities|Array|MakeArray (Variables|ParticipantInfo|GetTargetPawn) (Variables|ParticipantInfo|GetOwningPlayerPawn))
    (Utilities|IsValid _array_element
      (:Is%20Valid
        (bind _returnvalue (Fortnite|Utilities|AttemptGetPlayerStateforPawn _array_element))
        (bind _asfort_player_state_athena (Utilities|Casting|CastToFortPlayerStateAthena _returnvalue))
        (bind _returnvalue_2 (Class|FortGameStateAthena|GetSquadMemberPlayerStates (Fortnite|Athena|GameState|GetGameStateAthenafromContext _array_element) _asfort_player_state_athena true))
        (if (Utilities|Array|IsEmpty (LocalVariables|GetAllPlayerSquadStates))
          (LocalVariables|SetAllPlayerSquadStates _returnvalue_2)
          (else
            (Utilities|Array|AppendArray (LocalVariables|GetAllPlayerSquadStates) _returnvalue_2))))))
  (return (LocalVariables|GetAllPlayerSquadStates)))

(fn GetPlayerStatesForParticipants
  (params (InstigatorPawn :type Fort%20Player%20Pawn%20Object%20Reference) (TargetPawn :type Fort%20Player%20Pawn%20Object%20Reference))
  ;[0 0 0] Helper function to get player states from pawns
  (bind _returnvalue (Fortnite|Utilities|AttemptGetPlayerStateforPawn InstigatorPawn))
  (bind _asfort_player_state_athena (Utilities|Casting|CastToFortPlayerStateAthena _returnvalue))
  (bind _returnvalue_2 (Fortnite|Utilities|AttemptGetPlayerStateforPawn TargetPawn))
  (bind _asfort_player_state_athena_2 (Utilities|Casting|CastToFortPlayerStateAthena _returnvalue_2))
  (return _asfort_player_state_athena_2 _asfort_player_state_athena))

(fn HandleDuelPoachedOrFailed
  ;[0 0 0] Handle poached/failed logic
  (Variables|DuelBountiesVariables|SetPoachOrSuccessVerbTag "(TagName=\"CustomVerb.JobBoard.Poached\")")
  (CallFunction|FailQuest)
  (return))

(fn HandleDuelSuccess
  (bind _owningplayerpawn (Variables|ParticipantInfo|GetOwningPlayerPawn))
  (bind _returnvalue (Fortnite|GetFortPlayerControllerfromActor _owningplayerpawn))
  (bind _returnvalue_2 (Quests|GetQuestManager _returnvalue "Athena"))
  ;[0 0 0] Send quest event for the duel success
  (Variables|DuelBountiesVariables|SetPoachOrSuccessVerbTag "(TagName=\"CustomVerb.PlayerDuel.TargetEliminated\")")
  (Utilities|IsValid _owningplayerpawn
    (:Is%20Valid
      (Utilities|IsValid _returnvalue
        (:Is%20Valid
          (Utilities|IsValid _returnvalue_2
            (:Is%20Valid
              (Quest|SendComplexCustomStatEvent _returnvalue_2 0 "(GameplayTags=)" "(GameplayTags=((TagName=\"Athena.Quests.Bounty.Hunter\")))" 1)
              (CallFunction|CompleteQuest)
              (return))))))))

(fn MakeTrackingRadiusForTarget
  (params (TargetPawn :type Fort%20Player%20Pawn%20Object%20Reference) (MinRadius :type Float%20%28double-precision%29) (MaxRadius :type Float%20%28double-precision%29) (AmountToShrinkPerUpdate :type Float%20%28double-precision%29))
  (bind _numtrackingcircleupdates (Variables|HotfixableVariables|GetNumTrackingCircleUpdates))
  (bind _returnvalue (ScalableFloat|GetValueatLevel _numtrackingcircleupdates))
  ;[0 0 0] Make tracking radius for the duel target
  (if (> _returnvalue 0.9)
    (bind _returnvalue_2 (RadiusTracker|MakeRadiusTrackingInitializer (Math|Random|RandomFloatinRange MinRadius (* MaxRadius 0.75)) MinRadius MaxRadius (/ (ScalableFloat|GetValueatLevel (Variables|HotfixableVariables|GetDuelBountiesTime)) _returnvalue) AmountToShrinkPerUpdate "(R=1.000000,G=0.456411,B=0.000000,A=0.400000)" (LocalVariables|GetCircle%20Drawing%20Material) TargetPawn (Utilities|Array|MakeArray (Fortnite|GetFortPlayerControllerfromActor (Variables|ParticipantInfo|GetOwningPlayerPawn)))))
    (bind _returnvalue_3 (Fortnite|AddActorTrackingInfo (RadiusTracker|GetGameStateRadiusTrackerComponent) _returnvalue_2))
    (Variables|DuelBountiesVariables|SetRadiusTrackingGroupKey _returnvalue_3)))

(fn MakeUrgentQuestMessage
  (params (TitleText :type Text) (bSuccess :type Boolean))
  (bind _urgentquestreference (Variables|DuelBountiesVariables|GetUrgentQuestReference))
  ;[0 0 0] Makes a message to display on the urgent quest banner
  (Utilities|IsValid _urgentquestreference
    (:Is%20Valid
      (CallFunction|DisplayUrgentQuestMessage (Utilities|Struct|MakeUrgentQuestData (Utilities|Struct|BreakUrgentQuestData (Class|FortUrgentQuestItemDefinition|GetUrgentQuestData _urgentquestreference)) "(GameplayTags=)" -1.0 true false false false false true "" TitleText) bSuccess)
      (return))))

(fn OwnerIsDuelInstigator
  ;[0 0 0] Check if the ability owner is also the duel instigator (first ensure that the owner and the instigator references here are valid before comparing - FORT-789077)
  (Utilities|IsValid (Variables|ParticipantInfo|GetOwningPlayerPawn)
    (:Is%20Valid
      (Utilities|IsValid (Variables|ParticipantInfo|GetDuelInstigator)
        (:Is%20Valid
          (return (== (Variables|ParticipantInfo|GetOwningPlayerPawn) (Variables|ParticipantInfo|GetDuelInstigator))))
        (:Is%20Not%20Valid
          _)))
    (:Is%20Not%20Valid
      _))
  (return false))

(fn RemoveTrackingRadiusForTarget
  ;[0 0 0] Remove tracking radii from participants
  (Utilities|IsValid (RadiusTracker|GetGameStateRadiusTrackerComponent)
    (:Is%20Valid
      (if (CallFunction|OwnerIsDuelInstigator)
        (RadiusTracker|RemoveTrackingforActorbyTrackingKey (RadiusTracker|GetGameStateRadiusTrackerComponent) (Variables|ParticipantInfo|GetTargetPawn) (Variables|DuelBountiesVariables|GetRadiusTrackingGroupKey))
        (else
          (RadiusTracker|RemoveTrackingforActorfromPlayer (RadiusTracker|GetGameStateRadiusTrackerComponent) (Variables|ParticipantInfo|GetTargetPawn) (Fortnite|GetFortPlayerControllerfromActor (Variables|ParticipantInfo|GetOwningPlayerPawn)) "Squad"))))))

(fn SetPawnsFromPayload
  (params (PawnDamageInfo :type Pawn%20Damage%20Data%20Structure))
  (bind (_damagedactor _damage _instigatedby) (Utilities|Struct|BreakPawnDamageData PawnDamageInfo))
  (bind _returnvalue (Ability|GetAvatarActorfromActorInfo))
  (bind (_asfort_player_pawn_athena _bsuccess) (Utilities|Casting|CastToFortPlayerPawnAthena _returnvalue))
  (bind (_asfort_player_controller _bsuccess_2) (Utilities|Casting|CastToFortPlayerController _instigatedby))
  (bind (_asplayer_pawn_athena _bsuccess_3) (Utilities|Casting|CastToPlayerPawn_Athena _damagedactor))
  ;[0 0 0] Helper function to set references to the target, damager, and ability owner as Athena Pawns
  (Utilities|FlowControl|Sequence
    (:then_0
      ;;[0.1198 0.1198 0.1198] Check that damaged pawn and owner are valid Athena Pawns
      (if (and _bsuccess_3 _bsuccess)
        ;[0.151 0.151 0.151] Store references as Athena Pawns
        (Variables|ParticipantInfo|SetEliminatedPawnAthena _asplayer_pawn_athena)
        (Variables|ParticipantInfo|SetOwningPawnAthena _asfort_player_pawn_athena)
        (Variables|ParticipantInfo|SetbAnyPawnsValid true)
        (else
          (Variables|ParticipantInfo|SetbAnyPawnsValid false))))
    (:then_1
      ;;[0.1198 0.1198 0.1198] Check if the instigator is a valid pawn. Can be null (i.e. fall damage)
      (if _bsuccess_2
        (Variables|ParticipantInfo|SetDamagerPawnAthena (Utilities|Casting|CastToFortPlayerPawnAthena (Game|Player|GetPlayerPawn _asfort_player_controller)))
        (return)))))
