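; Gameplay ability for the reboot variant of a player duel quest (QS1) in Duel Bounties.
; On activation it sets the target-found event tag to the reboot trigger tag, then runs the parent ability's activation.
; Post-initialization and end-ability events only forward to the parent, so all other duel logic is inherited.
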(event Ability|EventActivateAbility
  (Variables|DuelBountiesVariables|SetTargetFoundEventTag (Variables|Default|GetRebootTriggerTag))
  (|CallParentFunction))

(event EventOnPostAbilityInitialization
  (|CallParentFunction))

(event Ability|EventOnEndAbility
  (params (bWasCancelled :type Boolean))
  (|CallParentFunction bWasCancelled))
